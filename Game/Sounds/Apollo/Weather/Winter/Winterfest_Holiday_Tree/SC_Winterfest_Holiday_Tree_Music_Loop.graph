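# Sound cue for looping diegetic music on the Winterfest holiday tree, drawing on five loops, several named for different seasons.
# A looping node wraps an equally weighted random pick, with single-input mixers boosting two of the tracks to different levels.
# Overridden attenuation applies logarithmic distance falloff and listener focus, so volume depends on where the listener faces.
# It uses the loot ambient loop sound class and loot ambient concurrency set.
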SoundCue /Game/Sounds/Apollo/Weather/Winter/Winterfest_Holiday_Tree/SC_Winterfest_Holiday_Tree_Music_Loop
root=Looping_0
nodes [
 {id=Looping_0, children [Random_0]}
 {id=Mixer_0, children [WavePlayer_2,null], properties {inputVolume [4.3,1]}}
 {id=Mixer_1, children [WavePlayer_5,null], properties {inputVolume [1.8,1]}}
 {id=Random_0, children [Mixer_0,WavePlayer_1,WavePlayer_3,WavePlayer_4,Mixer_1], properties {weights [1,1,1,1,1]}}
 {id=WavePlayer_1, properties {soundWaveAssetPtr=&/s28_Winterfest_Diegetic_Loop_03}}
 {
  id=WavePlayer_2
  properties {soundWaveAssetPtr=&/Winterfest_Holiday_Tree_Music_Loop_01}
 }
 {id=WavePlayer_3, properties {soundWaveAssetPtr=&/s23_Winterfest_Diegetic_Loop}}
 {id=WavePlayer_4, properties {soundWaveAssetPtr=&/MX_S39_Winterfest_Diegetic_Loop}}
 {id=WavePlayer_5, properties {soundWaveAssetPtr=&/S15_Winterfest_Diegetic_Loop}}
]
properties {
 bOverrideAttenuation=true
 maxDistance=1312
 soundClassObject=/Game/Sounds/FortSoundClassesAndModes/Fort_Loot_Ambient_Loop
 volumeMultiplier=0.4
 attenuationOverrides {
  bEnableFocusInterpolation=true
  bEnableListenerFocus=true
  distanceAlgorithm=Logarithmic
  falloffDistance=800
  focusAttackInterpSpeed=0.5
  focusAzimuth=20
  focusReleaseInterpSpeed=1.5
  focusVolumeAttenuation=1.2
  lPFFrequencyAtMax=10000
  lPFRadiusMax=1000
  lPFRadiusMin=500
  nonFocusAzimuth=40
  nonFocusVolumeAttenuation=0.5
  nonSpatializedRadiusStart=200
  occlusionInterpolationTime=0.3
  occlusionLowPassFilterFrequency=3000
  occlusionVolumeAttenuation=0.75
  stereoSpread=512
  attenuationShapeExtents {x=512}
  customAttenuationCurve {editorCurveData {keys [<at 0 1 RCIM_Cubic>, {arriveTangent=-1, interpMode=RCIM_Cubic, leaveTangent=-1, time=0.5, value=0.1}, <at 1 0 RCIM_Cubic>]}}
 }
 concurrencyOverrides {maxCount=1}
 concurrencySet [/Game/Sounds/Concurrency_Rules/Loot_Ambient_Conc]
}
